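# Physics asset for the female medium Sarong body cosmetic, providing bodies and constraints for the dress's front and back bones.
# A three-box back chain and two-box front chain link to tiny master capsules under the pelvis, with limited swing and tight twist.
# Capsules on the pelvis, right thigh and both thigh twists form the body shapes, joined by free angular constraints.
# Left-side front, side and back master capsules link to the left thigh twist but carry no chain bodies.
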PhysicsAsset /BRCosmetics/Characters/Player/Female/Medium/Bodies/F_MED_Sarong/Meshes/F_MED_Sarong_Physics
bodies {
 dyn_dress_back_01 [{shapeType=Box, extentX=6, extentY=3.1, extentZ=10.5, center=<3.1404 0 0>}]
 dyn_dress_back_01_MSTR [{shapeType=Capsule, length=0.01, radius=0.01}]
 dyn_dress_back_01_l_MSTR [{shapeType=Capsule, length=0.01, radius=0.01}]
 dyn_dress_back_02 [{shapeType=Box, extentX=6, extentY=3.1, extentZ=10.5, center=<3.1404 0 0>}]
 dyn_dress_back_03 [{shapeType=Box, extentX=6, extentY=3.1, extentZ=10.5, center=<3.1404 0 0>}]
 dyn_dress_front_01 [{shapeType=Box, extentX=8.5, extentY=2.1, extentZ=10.5, center=<4.4756 0.0001 0>}]
 dyn_dress_front_01_MSTR [{shapeType=Capsule, length=0.01, radius=0.01}]
 dyn_dress_front_01_l_MSTR [{shapeType=Capsule, length=0.01, radius=0.01}]
 dyn_dress_front_02 [{shapeType=Box, extentX=8.5, extentY=2.1, extentZ=10.5, center=<4.4756 0.0001 0>}]
 dyn_dress_side_01_l_MSTR [{shapeType=Capsule, length=0.01, radius=0.01}]
 pelvis [{shapeType=Capsule, length=15.3844, radius=8.4081, center=<-4.6018 -2.0818 0.0001>, rotation=<rot -0.0006 0 0>}]
 thigh_r [{shapeType=Capsule, length=33.0738, radius=9.3119, center=<19.3408 1.6453 0.0001>, rotation=<rot 90 -0.0049 179.9951>}]
 thigh_twist_01_l [{shapeType=Capsule, length=36.8238, radius=6.8119, center=<6.035 0.0011 0.0009>, rotation=<rot 90 0.0099 0.0099>}]
 thigh_twist_01_r [{shapeType=Capsule, length=36.8238, radius=6.8119, center=<-6.035 -0.0011 -0.0009>, rotation=<rot 90 -0.0099 179.9901>}]
}
constraints [
 {bone1Name=dyn_dress_back_01, bone2Name=dyn_dress_back_01_MSTR, swing1LimitDegrees=45, swing2LimitDegrees=25, twistLimitDegrees=5}
 {bone1Name=dyn_dress_back_02, bone2Name=dyn_dress_back_01, swing1LimitDegrees=45, swing2LimitDegrees=25, twistLimitDegrees=5}
 {bone1Name=dyn_dress_back_03, bone2Name=dyn_dress_back_02, swing1LimitDegrees=45, swing2LimitDegrees=25, twistLimitDegrees=5}
 {
  bone1Name=thigh_r
  bone2Name=pelvis
  swing1LimitDegrees=45
  swing1Motion=Free
  swing2LimitDegrees=45
  swing2Motion=Free
  twistLimitDegrees=45
  twistMotion=Free
 }
 {
  bone1Name=dyn_dress_front_01_l_MSTR
  bone2Name=thigh_twist_01_l
  swing1LimitDegrees=45
  swing1Motion=Free
  swing2LimitDegrees=45
  swing2Motion=Free
  twistLimitDegrees=45
  twistMotion=Free
 }
 {
  bone1Name=dyn_dress_side_01_l_MSTR
  bone2Name=thigh_twist_01_l
  swing1LimitDegrees=45
  swing1Motion=Free
  swing2LimitDegrees=45
  swing2Motion=Free
  twistLimitDegrees=45
  twistMotion=Free
 }
 {
  bone1Name=dyn_dress_back_01_l_MSTR
  bone2Name=thigh_twist_01_l
  swing1LimitDegrees=45
  swing1Motion=Free
  swing2LimitDegrees=45
  swing2Motion=Free
  twistLimitDegrees=45
  twistMotion=Free
 }
 {
  bone1Name=dyn_dress_front_01_MSTR
  bone2Name=pelvis
  swing1LimitDegrees=45
  swing1Motion=Free
  swing2LimitDegrees=45
  swing2Motion=Free
  twistLimitDegrees=45
  twistMotion=Free
 }
 {
  bone1Name=dyn_dress_back_01_MSTR
  bone2Name=pelvis
  swing1LimitDegrees=45
  swing1Motion=Free
  swing2LimitDegrees=45
  swing2Motion=Free
  twistLimitDegrees=45
  twistMotion=Free
 }
 {bone1Name=dyn_dress_front_01, bone2Name=dyn_dress_front_01_MSTR, swing1LimitDegrees=45, swing2LimitDegrees=45, twistLimitDegrees=5}
 {bone1Name=dyn_dress_front_02, bone2Name=dyn_dress_front_01, swing1LimitDegrees=45, swing2LimitDegrees=45, twistLimitDegrees=5}
 {
  bone1Name=thigh_twist_01_l
  bone2Name=pelvis
  swing1LimitDegrees=45
  swing1Motion=Free
  swing2LimitDegrees=45
  swing2Motion=Free
  twistLimitDegrees=45
  twistMotion=Free
 }
 {
  bone1Name=thigh_twist_01_r
  bone2Name=thigh_r
  swing1LimitDegrees=45
  swing1Motion=Free
  swing2LimitDegrees=45
  swing2Motion=Free
  twistLimitDegrees=45
  twistMotion=Free
 }
]
