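# Post-process style Anim Blueprint for the female medium Sarong body, copying the parent mesh pose and adding shoulder correctives and skirt secondary motion.
# Thigh constraints, chained AnimDynamics and a pelvis-space RigidBody node drive the skirt bones.
# Crouch, skydive, emote and custom-animation states weight corrective offsets and constraints.
# High LOD or disabled rigid bodies adds fixed corrections to the cached, already simulated pose.
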AnimBlueprint /BRCosmetics/Characters/Player/Female/Medium/Bodies/F_MED_Sarong/Meshes/F_MED_Sarong_AnimBP
target_skeleton=&/F_MED_Sarong_Skeleton
graphs [
 AnimGraph {nodes [
   "Constraint - dyn_dress_front_01_l_MSTR (P)"(Alpha=1, ConstraintWeights_0=1) {
    alphaBoolBlend {blendOption=Linear}
    boneToModify {boneName=dyn_dress_front_01_l_MSTR}
    constraintSetup [{targetBone {boneName=thigh_twist_01_l}}]
    constraintWeights [1]
   }
   n1 = "Constraint - thigh_twist_01_l (P)"(ComponentPose=Local To Component(LocalPose=Use cached pose 'lod0Cache'()), Alpha=1, ConstraintWeights_0=1) {
    alphaBoolBlend {blendOption=Linear}
    boneToModify {boneName=thigh_twist_01_l}
    constraintSetup [{targetBone {boneName=thigh_l}}]
    constraintWeights [1]
   }
   n2 = "Transform (Modify) Bone - Bone: dyn_dress_back_01"!(ComponentPose=@n1, bAlphaBoolEnabled:bool=True, Translation="0,-5,0", Rotation="0,0,0", Scale="1,1,1") -> (Pose) {
    alphaInputType=Bool
    translationMode=BMM_Additive
    translationSpace=BCS_BoneSpace
    alphaBoolBlend {blendOption=Linear}
    boneToModify {boneName=dyn_dress_back_01}
    translation {y=-5}
   }
   n3 = "Transform (Modify) Bone - Bone: dyn_dress_back_01_l"!(ComponentPose=@n2, bAlphaBoolEnabled:bool=True, Translation="0,-8,3", Rotation="0,0,0", Scale="1,1,1") -> (Pose) {
    alphaInputType=Bool
    translationMode=BMM_Additive
    translationSpace=BCS_BoneSpace
    alphaBoolBlend {blendOption=Linear}
    boneToModify {boneName=dyn_dress_back_01_l}
    translation {y=-8, z=3}
   }
   n4 = Blend Poses by bool(
    BlendPose_0=Component To Local(ComponentPose=@n3)
    BlendPose_1=Use cached pose 'lod0Cache'()
    bActiveValue=Get isMobileLod() -> (isMobileLod:bool)
   ) {blendType=Linear}
   Output Pose(Result=Component To Local(ComponentPose=Local To Component(LocalPose=@n4))) {layerGroup=None}
   n5 = "Transform (Modify) Bone - Bone: flont_uppelalm_l"(
    ComponentPose=Local To Component(LocalPose=Copy Pose From Mesh() {bUseAttachedParent=true})
    Alpha=Get Shoulder_rot_f_l() -> (Shoulder_rot_f_l:double)
    Translation="5,5,-5"
    Rotation="0,0,0"
    Scale="1,1,1"
   ) {
    rotationMode=BMM_Additive
    rotationSpace=BCS_BoneSpace
    scaleMode=BMM_Additive
    scaleSpace=BCS_BoneSpace
    translationMode=BMM_Additive
    translationSpace=BCS_BoneSpace
    alphaBoolBlend {blendOption=Linear}
    boneToModify {boneName=flont_uppelalm_l}
    translation=<5 5 -5>
   }
   n6 = "Transform (Modify) Bone - Bone: front_upperarm_r"(ComponentPose=@n5, Alpha=Get Shoulder_rot_f_r() -> (Shoulder_rot_f_r:double), Translation="-5,-5,5", Rotation="0,0,0", Scale="1,1,1") {
    rotationMode=BMM_Additive
    rotationSpace=BCS_BoneSpace
    scaleMode=BMM_Additive
    scaleSpace=BCS_BoneSpace
    translationMode=BMM_Additive
    translationSpace=BCS_BoneSpace
    alphaBoolBlend {blendOption=Linear}
    boneToModify {boneName=front_upperarm_r}
    translation=<-5 -5 5>
   }
   n7 = "Transform (Modify) Bone - Bone: up_uppelalm_l"(ComponentPose=@n6, Alpha=Get Shoulder_rot_up_l() -> (Shoulder_rot_up_l:double), Translation="0,2,2", Rotation="0,0,0", Scale="1,1,1") {
    rotationMode=BMM_Additive
    rotationSpace=BCS_BoneSpace
    scaleMode=BMM_Additive
    scaleSpace=BCS_BoneSpace
    translationMode=BMM_Additive
    translationSpace=BCS_BoneSpace
    alphaBoolBlend {blendOption=Linear}
    boneToModify {boneName=up_uppelalm_l}
    translation {y=2, z=2}
   }
   n8 = "Transform (Modify) Bone - Bone: up_upperarm_r"(ComponentPose=@n7, Alpha=Get Shoulder_rot_up_r() -> (Shoulder_rot_up_r:double), Translation="0,-2,-2", Rotation="0,0,0", Scale="1,1,1") {
    rotationMode=BMM_Additive
    rotationSpace=BCS_BoneSpace
    scaleMode=BMM_Additive
    scaleSpace=BCS_BoneSpace
    translationMode=BMM_Additive
    translationSpace=BCS_BoneSpace
    alphaBoolBlend {blendOption=Linear}
    boneToModify {boneName=up_upperarm_r}
    translation {y=-2, z=-2}
   }
   n9 = "Constraint - dyn_dress_side_01_l_MSTR (P)"(ComponentPose=@n8, Alpha=1, ConstraintWeights_0=1) {
    alphaBoolBlend {blendOption=Linear}
    boneToModify {boneName=dyn_dress_side_01_l_MSTR}
    constraintSetup [{targetBone {boneName=thigh_twist_01_l}}]
    constraintWeights [1]
   }
   n10 = "Constraint - dyn_dress_back_01_l_MSTR (P)"(ComponentPose=@n9, Alpha=Get skydiveAlpha() -> (skydiveAlpha:double), ConstraintWeights_0=1) {
    alphaBoolBlend {blendOption=Linear}
    boneToModify {boneName=dyn_dress_back_01_l_MSTR}
    constraintSetup [{targetBone {boneName=thigh_twist_01_l}}]
    constraintWeights [1]
   }
   Save cached pose 'lod0Cache1'(Pose=Component To Local(ComponentPose=@n10)) {cacheName=lod0Cache1}
   n11 = "Transform (Modify) Bone - Bone: dyn_dress_front_01_l_MSTR"(ComponentPose=Local To Component(LocalPose=Use cached pose 'lod0Cache1'()), Alpha=1, Translation="0,0,0", Rotation="0,-5,30", Scale="1,1,1") {
    rotationMode=BMM_Additive
    rotationSpace=BCS_BoneSpace
    translationMode=BMM_Additive
    translationSpace=BCS_BoneSpace
    alphaBoolBlend {blendOption=Linear}
    boneToModify {boneName=dyn_dress_front_01_l_MSTR}
    rotation {roll=30, yaw=-5}
   }
   n12 = Anim Dynamics - Bone: dyn_dress_front_01_l(ComponentPose=@n11, Alpha=0.75, ExternalForce="0,0,0") {
    angularDampingOverride=0.9999
    angularSpringConstant=50
    bAngularSpring=true
    bGravityOverrideInSimSpace=true
    bOverrideAngularDamping=true
    gravityScale=0.42
    simulationSpace=RootRelative
    alphaBoolBlend {blendOption=Linear}
    boundBone {boneName=dyn_dress_front_01_l}
    componentAppliedLinearAccClamp=<3000 3000 3000>
    componentLinearAccScale=<0.6 0.6 0.6>
    componentLinearVelScale=<0.6 0.6 0.6>
    gravityOverride {z=-980}
    physicsBodyDefinitions [{
      sphereCollisionRadius=0
      boundBone {boneName=dyn_dress_front_01_l}
      boxExtents=<6 2 12>
      constraintSetup {angularLimitsMax=<0 0 50>, angularLimitsMin=<0 0 0>, angularTarget=<1 0 0>}
      localJointOffset=<1 0 0>
     }]
    relativeSpaceBone {boneName=pelvis}
   }
   n13 = Anim Dynamics - Bone: dyn_dress_front_02_l(ComponentPose=@n12, Alpha=0.75, ExternalForce="0,0,0") {
    angularDampingOverride=0.9999
    angularSpringConstant=50
    bAngularSpring=true
    bGravityOverrideInSimSpace=true
    bOverrideAngularDamping=true
    gravityScale=0.42
    simulationSpace=RootRelative
    alphaBoolBlend {blendOption=Linear}
    boundBone {boneName=dyn_dress_front_02_l}
    componentAppliedLinearAccClamp=<3000 3000 3000>
    componentLinearAccScale=<0.6 0.6 0.6>
    componentLinearVelScale=<0.6 0.6 0.6>
    physicsBodyDefinitions [{
      sphereCollisionRadius=0
      boundBone {boneName=dyn_dress_front_02_l}
      boxExtents=<6 2 12>
      constraintSetup {angularLimitsMax=<0 0 50>, angularLimitsMin=<0 0 0>, angularTarget=<1 0 0>}
      localJointOffset=<1 0 0>
     }]
    relativeSpaceBone {boneName=pelvis}
   }
   n14 = Anim Dynamics - Bone: dyn_dress_front_03_l(ComponentPose=@n13, Alpha=0.75, ExternalForce="0,0,0") {
    angularDampingOverride=0.9999
    angularSpringConstant=50
    bAngularSpring=true
    bGravityOverrideInSimSpace=true
    bOverrideAngularDamping=true
    gravityScale=0.42
    simulationSpace=RootRelative
    alphaBoolBlend {blendOption=Linear}
    boundBone {boneName=dyn_dress_front_03_l}
    componentAppliedLinearAccClamp=<3000 3000 3000>
    componentLinearAccScale=<0.6 0.6 0.6>
    componentLinearVelScale=<0.6 0.6 0.6>
    physicsBodyDefinitions [{
      sphereCollisionRadius=0
      boundBone {boneName=dyn_dress_front_03_l}
      boxExtents=<6 2 12>
      constraintSetup {angularLimitsMax=<0 0 50>, angularLimitsMin=<0 0 0>, angularTarget=<1 0 0>}
      localJointOffset=<1 0 0>
     }]
    relativeSpaceBone {boneName=pelvis}
   }
   n15 = Component To Local(ComponentPose=@n14)
   n16 = Anim Dynamics - Bone: dyn_dress_front_01_l(ComponentPose=Local To Component(LocalPose=Use cached pose 'lod0Cache1'()), Alpha=0.75, ExternalForce="0,0,0") {
    angularDampingOverride=0.9999
    angularSpringConstant=50
    bAngularSpring=true
    bGravityOverrideInSimSpace=true
    bOverrideAngularDamping=true
    gravityScale=0.42
    simulationSpace=RootRelative
    alphaBoolBlend {blendOption=Linear}
    boundBone {boneName=dyn_dress_front_01_l}
    componentAppliedLinearAccClamp=<3000 3000 3000>
    componentLinearAccScale=<0.6 0.6 0.6>
    componentLinearVelScale=<0.6 0.6 0.6>
    physicsBodyDefinitions [{
      sphereCollisionRadius=0
      boundBone {boneName=dyn_dress_front_01_l}
      boxExtents=<6 2 12>
      constraintSetup {angularLimitsMax=<0 0 50>, angularLimitsMin=<0 0 5>, angularTarget=<1 0 0>}
      localJointOffset=<1 0 0>
     }]
    relativeSpaceBone {boneName=pelvis}
   }
   n17 = Anim Dynamics - Bone: dyn_dress_front_02_l(ComponentPose=@n16, Alpha=0.75, ExternalForce="0,0,0") {
    angularDampingOverride=0.9999
    angularSpringConstant=50
    bAngularSpring=true
    bGravityOverrideInSimSpace=true
    bOverrideAngularDamping=true
    gravityScale=0.42
    simulationSpace=RootRelative
    alphaBoolBlend {blendOption=Linear}
    boundBone {boneName=dyn_dress_front_02_l}
    componentAppliedLinearAccClamp=<3000 3000 3000>
    componentLinearAccScale=<0.6 0.6 0.6>
    componentLinearVelScale=<0.6 0.6 0.6>
    physicsBodyDefinitions [{
      sphereCollisionRadius=0
      boundBone {boneName=dyn_dress_front_02_l}
      boxExtents=<6 2 12>
      constraintSetup {angularLimitsMax=<0 0 50>, angularLimitsMin=<0 0 3>, angularTarget=<1 0 0>}
      localJointOffset=<1 0 0>
     }]
    relativeSpaceBone {boneName=pelvis}
   }
   n18 = Anim Dynamics - Bone: dyn_dress_front_03_l(ComponentPose=@n17, Alpha=0.75, ExternalForce="0,0,0") {
    angularDampingOverride=0.9999
    angularSpringConstant=50
    bAngularSpring=true
    bGravityOverrideInSimSpace=true
    bOverrideAngularDamping=true
    gravityScale=0.42
    simulationSpace=RootRelative
    alphaBoolBlend {blendOption=Linear}
    boundBone {boneName=dyn_dress_front_03_l}
    componentAppliedLinearAccClamp=<3000 3000 3000>
    componentLinearAccScale=<0.6 0.6 0.6>
    componentLinearVelScale=<0.6 0.6 0.6>
    physicsBodyDefinitions [{
      sphereCollisionRadius=0
      boundBone {boneName=dyn_dress_front_03_l}
      boxExtents=<6 2 12>
      constraintSetup {angularLimitsMax=<0 0 50>, angularLimitsMin=<0 0 3>, angularTarget=<1 0 0>}
      localJointOffset=<1 0 0>
     }]
    relativeSpaceBone {boneName=pelvis}
   }
   n19 = "Transform (Modify) Bone - Bone: dyn_dress_back_02_l"!(
    ComponentPose=Local To Component(LocalPose=Component To Local(ComponentPose=@n18))
    bAlphaBoolEnabled:bool=True
    Translation="0,-5,6"
    Rotation="0,0,0"
    Scale="1,1,1"
   ) -> (Pose) {
    alphaInputType=Bool
    translationMode=BMM_Additive
    translationSpace=BCS_BoneSpace
    alphaBoolBlend {blendInTime=1, blendOption=Linear, blendOutTime=2}
    boneToModify {boneName=dyn_dress_back_02_l}
    translation {y=-5, z=6}
   }
   n20 = "Transform (Modify) Bone - Bone: dyn_dress_back_03_l"!(ComponentPose=@n19, bAlphaBoolEnabled:bool=True, Translation="0,-5,6", Rotation="0,0,0", Scale="1,1,1") -> (Pose) {
    alphaInputType=Bool
    translationMode=BMM_Additive
    translationSpace=BCS_BoneSpace
    alphaBoolBlend {blendInTime=1, blendOption=Linear, blendOutTime=2}
    boneToModify {boneName=dyn_dress_back_03_l}
    translation {y=-5, z=6}
   }
   n21 = Component To Local(ComponentPose=@n20)
   n22 = Get bIsCrouching()
   n23 = Anim Dynamics - Bone: dyn_dress_side_01_l(
    ComponentPose=Local To Component(LocalPose=Blend Poses by bool(BlendPose_0=@n15, BlendPose_1=@n21, BlendTime_0=0.5, BlendTime_1=0.2, bActiveValue=@n22) {blendType=Linear})
    Alpha=0.5
    ExternalForce="0,0,0"
   ) {
    angularDampingOverride=0.9999
    angularSpringConstant=50
    bAngularSpring=true
    bGravityOverrideInSimSpace=true
    bOverrideAngularDamping=true
    gravityScale=0.35
    simulationSpace=RootRelative
    alphaBoolBlend {blendOption=Linear}
    boundBone {boneName=dyn_dress_side_01_l}
    componentAppliedLinearAccClamp=<3000 3000 3000>
    componentLinearAccScale=<0.6 0.6 0.6>
    componentLinearVelScale=<0.6 0.6 0.6>
    physicsBodyDefinitions [{
      sphereCollisionRadius=0
      boundBone {boneName=dyn_dress_side_01_l}
      boxExtents=<6 2 12>
      constraintSetup {angularLimitsMax=<0 0 0>, angularLimitsMin=<0 -15 0>, angularTarget=<1 0 0>}
      localJointOffset=<1 0 0>
     }]
    relativeSpaceBone {boneName=pelvis}
   }
   n24 = Anim Dynamics - Bone: dyn_dress_side_02_l(ComponentPose=@n23, Alpha=0.5, ExternalForce="0,0,0") {
    angularDampingOverride=0.9999
    angularSpringConstant=50
    bAngularSpring=true
    bGravityOverrideInSimSpace=true
    bOverrideAngularDamping=true
    gravityScale=0.35
    simulationSpace=RootRelative
    alphaBoolBlend {blendOption=Linear}
    boundBone {boneName=dyn_dress_side_02_l}
    componentAppliedLinearAccClamp=<3000 3000 3000>
    componentLinearAccScale=<0.6 0.6 0.6>
    componentLinearVelScale=<0.6 0.6 0.6>
    physicsBodyDefinitions [{
      sphereCollisionRadius=0
      boundBone {boneName=dyn_dress_side_02_l}
      boxExtents=<6 2 12>
      constraintSetup {angularLimitsMax=<0 0 0>, angularLimitsMin=<0 -45 0>, angularTarget=<1 0 0>}
      localJointOffset=<1 0 0>
     }]
    relativeSpaceBone {boneName=pelvis}
   }
   n25 = Anim Dynamics - Bone: dyn_dress_side_03_l(ComponentPose=@n24, Alpha=Get crouchAlphaInv() -> (crouchAlphaInv:double), ExternalForce="0,0,0") {
    angularDampingOverride=0.9999
    angularSpringConstant=50
    bAngularSpring=true
    bGravityOverrideInSimSpace=true
    bOverrideAngularDamping=true
    gravityScale=0.35
    simulationSpace=RootRelative
    alphaBoolBlend {blendOption=Linear}
    boundBone {boneName=dyn_dress_side_03_l}
    componentAppliedLinearAccClamp=<3000 3000 3000>
    componentLinearAccScale=<0.6 0.6 0.6>
    componentLinearVelScale=<0.6 0.6 0.6>
    physicsBodyDefinitions [{
      sphereCollisionRadius=0
      boundBone {boneName=dyn_dress_side_03_l}
      boxExtents=<6 2 12>
      constraintSetup {angularLimitsMax=<0 0 0>, angularLimitsMin=<0 -45 0>, angularTarget=<1 0 0>}
      localJointOffset=<1 0 0>
     }]
    relativeSpaceBone {boneName=pelvis}
   }
   n26 = Anim Dynamics - Bone: dyn_dress_back_01_l(ComponentPose=@n25, Alpha=1, ExternalForce="0,0,0") {
    angularDampingOverride=0.9999
    angularSpringConstant=50
    bAngularSpring=true
    bGravityOverrideInSimSpace=true
    bOverrideAngularDamping=true
    gravityScale=0.35
    simulationSpace=RootRelative
    alphaBoolBlend {blendOption=Linear}
    boundBone {boneName=dyn_dress_back_01_l}
    componentAppliedLinearAccClamp=<3000 3000 3000>
    componentLinearAccScale=<0.6 0.6 0.6>
    componentLinearVelScale=<0.6 0.6 0.6>
    physicsBodyDefinitions [{
      sphereCollisionRadius=0
      boundBone {boneName=dyn_dress_back_01_l}
      boxExtents=<6 2 12>
      constraintSetup {angularLimitsMax=<0 0 0>, angularLimitsMin=<0 0 -50>, angularTarget=<1 0 0>}
      localJointOffset=<1 0 0>
     }]
    relativeSpaceBone {boneName=pelvis}
   }
   n27 = Anim Dynamics - Bone: dyn_dress_back_02_l(ComponentPose=@n26, Alpha=1, ExternalForce="0,0,0") {
    angularDampingOverride=0.9999
    angularSpringConstant=50
    bAngularSpring=true
    bGravityOverrideInSimSpace=true
    bOverrideAngularDamping=true
    gravityScale=0.35
    simulationSpace=RootRelative
    alphaBoolBlend {blendOption=Linear}
    boundBone {boneName=dyn_dress_back_02_l}
    componentAppliedLinearAccClamp=<3000 3000 3000>
    componentLinearAccScale=<0.6 0.6 0.6>
    componentLinearVelScale=<0.6 0.6 0.6>
    physicsBodyDefinitions [{
      sphereCollisionRadius=0
      boundBone {boneName=dyn_dress_back_02_l}
      boxExtents=<6 2 12>
      constraintSetup {angularLimitsMax=<0 0 0>, angularLimitsMin=<0 0 -50>, angularTarget=<1 0 0>}
      localJointOffset=<1 0 0>
     }]
    relativeSpaceBone {boneName=pelvis}
   }
   n28 = Anim Dynamics - Bone: dyn_dress_back_03_l(ComponentPose=@n27, Alpha=1, ExternalForce="0,0,0") {
    angularDampingOverride=0.9999
    angularSpringConstant=50
    bAngularSpring=true
    bGravityOverrideInSimSpace=true
    bOverrideAngularDamping=true
    gravityScale=0.35
    simulationSpace=RootRelative
    alphaBoolBlend {blendOption=Linear}
    boundBone {boneName=dyn_dress_back_03_l}
    componentAppliedLinearAccClamp=<3000 3000 3000>
    componentLinearAccScale=<0.6 0.6 0.6>
    componentLinearVelScale=<0.6 0.6 0.6>
    physicsBodyDefinitions [{
      sphereCollisionRadius=0
      boundBone {boneName=dyn_dress_back_03_l}
      boxExtents=<6 2 12>
      constraintSetup {angularLimitsMax=<0 0 0>, angularLimitsMin=<0 0 -50>, angularTarget=<1 0 0>}
      localJointOffset=<1 0 0>
     }]
    relativeSpaceBone {boneName=pelvis}
   }
   n29 = "Transform (Modify) Bone - Bone: dyn_dress_side_01_l_MSTR"(ComponentPose=@n28, Alpha=Get crouchAlphaInv(), Translation="0,0,0", Rotation="5,0,0", Scale="1,1,1") {rotationMode=BMM_Additive, alphaBoolBlend {blendOption=Linear}, boneToModify {boneName=dyn_dress_side_01_l_MSTR}, rotation {pitch=5}}
   n30 = Get skydiveAlpha()
   n31 = "Transform (Modify) Bone - Bone: dyn_dress_back_01_l_MSTR"(ComponentPose=@n29, Alpha=@n30, Translation="0,-3,0", Rotation="0,0,25", Scale="1,1,1") {
    rotationMode=BMM_Additive
    translationMode=BMM_Additive
    alphaBoolBlend {blendOption=Linear}
    boneToModify {boneName=dyn_dress_back_01_l_MSTR}
    rotation {roll=25}
    translation {y=-3}
   }
   n32 = "Transform (Modify) Bone - Bone: dyn_dress_side_01_l"(ComponentPose=@n31, Alpha=@n30, Translation="3,0,0", Rotation="-15,0,0", Scale="1,1,1") {
    rotationMode=BMM_Additive
    translationMode=BMM_Additive
    alphaBoolBlend {blendOption=Linear}
    boneToModify {boneName=dyn_dress_side_01_l}
    rotation {pitch=-15}
    translation {x=3}
   }
   n33 = "Transform (Modify) Bone - Bone: dyn_dress_side_02_l"(ComponentPose=@n32, Alpha=@n30, Translation="3,0,0", Rotation="5,0,0", Scale="1,1,1") {
    rotationMode=BMM_Additive
    translationMode=BMM_Additive
    alphaBoolBlend {blendOption=Linear}
    boneToModify {boneName=dyn_dress_side_02_l}
    rotation {pitch=5}
    translation {x=3}
   }
   n34 = Get crouchAlpha() -> (crouchAlpha:double)
   n35 = "Transform (Modify) Bone - Bone: dyn_dress_front_01_l_MSTR"(ComponentPose=@n33, Alpha=@n34, Translation="-3,0,1", Rotation="0,0,-5", Scale="1,1,1") {
    rotationMode=BMM_Additive
    translationMode=BMM_Additive
    alphaBoolBlend {blendOption=Linear}
    boneToModify {boneName=dyn_dress_front_01_l_MSTR}
    rotation {roll=-5}
    translation {x=-3, z=1}
   }
   n36 = "Transform (Modify) Bone - Bone: dyn_dress_front_01_MSTR"(ComponentPose=@n35, Alpha=@n34, Translation="0.000,-0.000,2.605", Rotation="0,0,0", Scale="1,1,1") {
    rotationMode=BMM_Additive
    translationMode=BMM_Additive
    alphaBoolBlend {blendOption=Linear}
    boneToModify {boneName=dyn_dress_front_01_MSTR}
    translation {z=2.605}
   }
   n37 = "Constraint - dyn_dress_front_01_MSTR (RR)"!(ComponentPose=@n36, Alpha=@n34, ConstraintWeights_0=1, ConstraintWeights_1:float=1) -> (Pose) {
    alphaBoolBlend {blendOption=Linear}
    boneToModify {boneName=dyn_dress_front_01_MSTR}
    constraintSetup [{transformType=Rotation, targetBone {boneName=thigh_l}}, {transformType=Rotation, targetBone {boneName=thigh_r}}]
    constraintWeights [1,1]
   }
   n38 = "Transform (Modify) Bone - Bone: dyn_dress_side_01_l"(ComponentPose=@n37, Alpha=@n34, Translation="0,0,0", Rotation="0,0,0", Scale="1,1,1") {rotationMode=BMM_Additive, translationMode=BMM_Additive, alphaBoolBlend {blendOption=Linear}, boneToModify {boneName=dyn_dress_side_01_l}}
   n39 = "Transform (Modify) Bone - Bone: dyn_dress_back_01_l"!(ComponentPose=@n38, Alpha:float=@n34, Translation="5,-10,0", Rotation="0,0,15", Scale="1,1,1") -> (Pose) {
    rotationMode=BMM_Additive
    translationMode=BMM_Additive
    alphaBoolBlend {blendOption=Linear}
    boneToModify {boneName=dyn_dress_back_01_l}
    rotation {roll=15}
    translation=<5 -10>
   }
   n40 = RigidBody(ComponentPose=@n39) {
    bClampLinearTranslationLimitToRefPose=true
    bOverrideWorldGravity=true
    overridePhysicsAsset=&/F_MED_Sarong_Physics
    simulationSpace=BaseBoneSpace
    alphaBoolBlend {blendOption=Linear}
    baseBoneRef {boneName=pelvis}
    componentAppliedLinearAccClamp=<3000 3000 3000>
    componentLinearAccScale=<0.45 0.45 0.45>
    componentLinearVelScale=<0.45 0.45 0.45>
    overrideWorldGravity {z=-980}
   }
   n41 = "Transform (Modify) Bone - Bone: dyn_dress_back_01_MSTR"(ComponentPose=@n40, Alpha=Get skydiveAlphaInv() -> (skydiveAlphaInv:double), Translation="0,-1.5,0", Rotation="0,0,10", Scale="1,1,1") {
    rotationMode=BMM_Additive
    translationMode=BMM_Additive
    alphaBoolBlend {blendOption=Linear}
    boneToModify {boneName=dyn_dress_back_01_MSTR}
    rotation {roll=10}
    translation {y=-1.5}
   }
   n42 = Get customAnimAlpha() -> (customAnimAlpha:double)
   n43 = "Transform (Modify) Bone - Bone: dyn_dress_front_01_l_MSTR"(ComponentPose=@n41, Alpha=@n42, Translation="0,2,0", Rotation="0,0,-5", Scale="1,1,1") {
    rotationMode=BMM_Additive
    translationMode=BMM_Additive
    alphaBoolBlend {blendOption=Linear}
    boneToModify {boneName=dyn_dress_front_01_l_MSTR}
    rotation {roll=-5}
    translation {y=2}
   }
   n44 = "Transform (Modify) Bone - Bone: dyn_dress_side_01_l_MSTR"(ComponentPose=@n43, Alpha=@n42, Translation="2,0,0", Rotation="-2,0,0", Scale="1,1,1") {
    rotationMode=BMM_Additive
    translationMode=BMM_Additive
    alphaBoolBlend {blendOption=Linear}
    boneToModify {boneName=dyn_dress_side_01_l_MSTR}
    rotation {pitch=-2}
    translation {x=2}
   }
   n45 = Get emotingAlpha() -> (emotingAlpha:double)
   n46 = "Constraint - dyn_dress_front_01_l_MSTR (P)"(ComponentPose=@n44, Alpha=@n45, ConstraintWeights_0=1) {
    alphaBoolBlend {blendOption=Linear}
    boneToModify {boneName=dyn_dress_front_01_l_MSTR}
    constraintSetup [{targetBone {boneName=thigh_l}}]
    constraintWeights [1]
   }
   n47 = "Constraint - dyn_dress_side_01_l_MSTR (P)"(ComponentPose=@n46, Alpha=@n45, ConstraintWeights_0=1) {
    alphaBoolBlend {blendOption=Linear}
    boneToModify {boneName=dyn_dress_side_01_l_MSTR}
    constraintSetup [{targetBone {boneName=thigh_l}}]
    constraintWeights [1]
   }
   n48 = "Constraint - dyn_dress_back_01_l_MSTR (P)"(ComponentPose=@n47, Alpha=@n45, ConstraintWeights_0=1) {
    alphaBoolBlend {blendOption=Linear}
    boneToModify {boneName=dyn_dress_back_01_l_MSTR}
    constraintSetup [{targetBone {boneName=pelvis}}]
    constraintWeights [1]
   }
   n49 = "Transform (Modify) Bone - Bone: dyn_dress_front_01_l_MSTR"(ComponentPose=@n48, Alpha=@n45, Translation="0,0,0", Rotation="0,0,-8", Scale="1,1,1") {
    rotationMode=BMM_Additive
    translationMode=BMM_Additive
    alphaBoolBlend {blendOption=Linear}
    boneToModify {boneName=dyn_dress_front_01_l_MSTR}
    rotation {roll=-8}
   }
   n50 = "Transform (Modify) Bone - Bone: dyn_dress_front_02_l"(ComponentPose=@n49, Alpha=@n45, Translation="0,0,0", Rotation="0,0,-8", Scale="1,1,1") {
    rotationMode=BMM_Additive
    translationMode=BMM_Additive
    alphaBoolBlend {blendOption=Linear}
    boneToModify {boneName=dyn_dress_front_02_l}
    rotation {roll=-8}
   }
   n51 = "Transform (Modify) Bone - Bone: dyn_dress_back_01_l_MSTR"(ComponentPose=@n50, Alpha=@n45, Translation="-2,0,0", Rotation="0,0,0", Scale="1,1,1") {
    rotationMode=BMM_Additive
    translationMode=BMM_Additive
    alphaBoolBlend {blendOption=Linear}
    boneToModify {boneName=dyn_dress_back_01_l_MSTR}
    translation {x=-2}
   }
   n52 = "Transform (Modify) Bone - Bone: dyn_dress_side_01_l_MSTR"(ComponentPose=@n51, Alpha=@n45, Translation="2,0,0", Rotation="0,-10,0", Scale="1,1,1") {
    rotationMode=BMM_Additive
    translationMode=BMM_Additive
    alphaBoolBlend {blendOption=Linear}
    boneToModify {boneName=dyn_dress_side_01_l_MSTR}
    rotation {yaw=-10}
    translation {x=2}
   }
   Save cached pose 'lod0Cache'(Pose=Component To Local(ComponentPose=@n52)) {cacheName=lod0Cache}
  ]}
 EventGraph {nodes [
   Event BlueprintUpdateAnimation()
   > n1 = Sequence!(execute) -> (then_0, then_1, then_2:Exec, then_3:Exec)
   n2 = Get bIsCrouching()
   > Set crouchAlphaInv(crouchAlphaInv:double="To Float (Boolean)"(InBool=NOT Boolean(A=@n2))) -> (Output_Get:double)
   > Set crouchAlpha(crouchAlpha:double="To Float (Boolean)"(InBool=@n2)) -> (Output_Get:double)
   n3 = Get bIsPlayingEmote()
   Set emotingAlpha(execute=@n1.then_1, emotingAlpha:double="To Float (Boolean)"(InBool=@n3)) -> (Output_Get:double)
   n4 = Get bCanPlayCustomAnimations() -> (bCanPlayCustomAnimations:bool)
   > Set customAnimAlpha(customAnimAlpha:double="To Float (Boolean)"(InBool=OR Boolean(A=@n3, B=@n4))) -> (Output_Get:double)
   n5 = Get CurrentLOD()
   n6 = Get bRigidBodyEnabled()
   Set isMobileLod(execute=@n1.then_2, isMobileLod:bool=OR Boolean(A=NOT Boolean(A=@n6), B="integer > integer"(A=@n5, B=1) {is=title K2Node_CallFunction})) -> (Output_Get:bool)
   n7 = Get Skel_Mesh()
   Set Shoulder_rot_f_r(
    execute=@n1.then_3
    Shoulder_rot_f_r:double=GetAlphaFromReferenceTransformDelta(SkeletalMeshComponent=@n7, BoneName=upperarm_r, BaseName=clavicle_r, InputRange=(X=0,Y=90), TwistAxis="0,0,1")
   ) -> (Output_Get:double)
   > Set Shoulder_rot_f_l(Shoulder_rot_f_l:double=GetAlphaFromReferenceTransformDelta(SkeletalMeshComponent=@n7, BoneName=upperarm_l, BaseName=clavicle_l, InputRange=(X=0,Y=90), TwistAxis="0,0,1")) -> (Output_Get:double)
   n8 = Get Skel_Mesh()
   > Set Shoulder_rot_up_r(Shoulder_rot_up_r:double=GetAlphaFromReferenceTransformDelta(
     SkeletalMeshComponent=@n8
     BoneName=upperarm_twist_01_r
     BaseName=clavicle_r
     DeltaComponentType=EulerY
     InputRange=(X=0,Y=110)
     TwistAxis="0,0,1"
    )) -> (Output_Get:double)
   > Set Shoulder_rot_up_l(Shoulder_rot_up_l:double=GetAlphaFromReferenceTransformDelta(SkeletalMeshComponent=@n8, BoneName=upperarm_l, BaseName=clavicle_l, DeltaComponentType=EulerY, InputRange=(X=0,Y=110), TwistAxis="0,0,1")) -> (Output_Get:double)
   Event OnBeginSkydiving()
   > Set skydiveAlpha(skydiveAlpha:double=1) -> (Output_Get:double)
   > Set skydiveAlphaInv(skydiveAlphaInv:double=0) -> (Output_Get:double)
   Event OnEndSkydiving()
   > Set skydiveAlpha(skydiveAlpha=0)
   > Set skydiveAlphaInv(skydiveAlphaInv=1)
   Event BlueprintInitializeAnimation()
   > Set Skel_Mesh(Skel_Mesh=GetOwnerSkeletalMeshComponent())
  ]}
]
variables [
 crouchAlpha {type=number}
 crouchAlphaInv {type=number}
 customAnimAlpha {type=number}
 emotingAlpha {type=number}
 isMobileLod {type=boolean}
 Shoulder_rot_f_l {type=number}
 Shoulder_rot_f_r {type=number}
 Shoulder_rot_up_l {type=number}
 Shoulder_rot_up_r {type=number}
 Skel_Mesh {type=SkeletalMeshComponent}
 skydiveAlpha {type=number}
 skydiveAlphaInv {type=number}
]
