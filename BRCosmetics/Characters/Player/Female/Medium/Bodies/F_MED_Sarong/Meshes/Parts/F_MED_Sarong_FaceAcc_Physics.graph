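# Physics asset for the face accessory part of the female medium Sarong player body, defining physics for its ponytail.
# Capsules on the head, neck, upper spine and clavicles form the body shape around the hair, linked by free angular constraints.
# A five-bone ponytail chain hangs from a tiny master anchor linked to the head.
# Its root joint has tighter swing limits, while later joints swing wider with zero twist allowance.
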PhysicsAsset /BRCosmetics/Characters/Player/Female/Medium/Bodies/F_MED_Sarong/Meshes/Parts/F_MED_Sarong_FaceAcc_Physics
bodies {
 clavicle_l [{shapeType=Capsule, length=5.6128, radius=4.2367, center=<7.1889 -0.3553 -1.4606>, rotation=<rot 76.6244 0 0>}]
 clavicle_r [{shapeType=Capsule, length=4.5521, radius=5.6963, center=<-6.8579 0.3553 1.3819>, rotation=<rot 76.6244 0 0>}]
 dyn_ponytail_01 [{
   shapeType=Capsule
   length=2.9211
   radius=2.8832
   shapeName=dyn_ponytail_01_capsule
   center=<2.6377 0 -0.0404>
   rotation=<rot -89.5912 179.9897 -179.9897>
  }]
 dyn_ponytail_01_MSTR [{shapeType=Capsule, length=0.01, radius=0.01}]
 dyn_ponytail_02 [{
   shapeType=Capsule
   length=3.6048
   radius=3.2349
   shapeName=dyn_ponytail_02_capsule
   center=<2.2231 0 -0.0113>
   rotation=<rot 89.0418 -179.9954 -179.9954>
  }]
 dyn_ponytail_03 [{
   shapeType=Capsule
   length=12.5176
   radius=2.8935
   shapeName=dyn_ponytail_03_capsule
   center=<4.7443 0.0001 -0.1745>
   rotation=<rot -87.8931 179.9996 -179.9998>
  }]
 dyn_ponytail_04 [{shapeType=Capsule, length=4.1937, radius=3.5611, shapeName=dyn_ponytail_04_capsule, center=<0.7293 -0.2496 0>}]
 dyn_ponytail_05 [{
   shapeType=Capsule
   length=12.5943
   radius=4.9861
   shapeName=dyn_ponytail_05_capsule
   center=<0.3098 -1.2158 0.0103>
   rotation=<rot 2.4613 0 0.0001>
  }]
 head [{shapeType=Capsule, length=3.9729, radius=8.3214, center=<6.4012 0 -0.0001>, rotation=<rot -90 -0.0396 0.0397>}]
 neck_02 [{shapeType=Capsule, length=12.3479, radius=6.3526, center=<2.1491 0.886 -0.0001>, rotation=<rot -90 90 -100.0009>}]
 spine_05 [{shapeType=Capsule, length=11.8717, radius=14.0151, center=<1.0025 6.238 -0.0017>, rotation=<rot 0 -0.0149 0.0149>}]
}
constraints [
 {
  bone1Name=head
  bone2Name=neck_02
  swing1LimitDegrees=45
  swing1Motion=Free
  swing2LimitDegrees=45
  swing2Motion=Free
  twistLimitDegrees=45
  twistMotion=Free
 }
 {
  bone1Name=neck_02
  bone2Name=spine_05
  swing1LimitDegrees=45
  swing1Motion=Free
  swing2LimitDegrees=45
  swing2Motion=Free
  twistLimitDegrees=45
  twistMotion=Free
 }
 {
  bone1Name=dyn_ponytail_01_MSTR
  bone2Name=head
  swing1LimitDegrees=45
  swing1Motion=Free
  swing2LimitDegrees=45
  swing2Motion=Free
  twistLimitDegrees=45
  twistMotion=Free
 }
 {bone1Name=dyn_ponytail_01, bone2Name=dyn_ponytail_01_MSTR, swing1LimitDegrees=30, swing2LimitDegrees=15, twistLimitDegrees=5}
 {bone1Name=dyn_ponytail_02, bone2Name=dyn_ponytail_01, swing1LimitDegrees=60, swing2LimitDegrees=30, twistLimitDegrees=0}
 {bone1Name=dyn_ponytail_03, bone2Name=dyn_ponytail_02, swing1LimitDegrees=60, swing2LimitDegrees=30, twistLimitDegrees=0}
 {bone1Name=dyn_ponytail_04, bone2Name=dyn_ponytail_03, swing1LimitDegrees=60, swing2LimitDegrees=30, twistLimitDegrees=0}
 {bone1Name=dyn_ponytail_05, bone2Name=dyn_ponytail_04, swing1LimitDegrees=60, swing2LimitDegrees=30, twistLimitDegrees=0}
 {
  bone1Name=clavicle_l
  bone2Name=spine_05
  swing1LimitDegrees=45
  swing1Motion=Free
  swing2LimitDegrees=45
  swing2Motion=Free
  twistLimitDegrees=45
  twistMotion=Free
 }
 {
  bone1Name=clavicle_r
  bone2Name=spine_05
  swing1LimitDegrees=45
  swing1Motion=Free
  swing2LimitDegrees=45
  swing2Motion=Free
  twistLimitDegrees=45
  twistMotion=Free
 }
]
