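# Post-copy AnimBP for the female Sarong face accessory, adding secondary motion to its flower and ponytail bones.
# It copies the parent mesh pose, runs AnimDynamics on flower bones and a neck-space RigidBody, then caches the result.
# When LOD exceeds 1 or rigid body is disabled, ponytail constraints to spine/head and an additive swing are layered on.
# Swing alpha follows head swing angle, or is full while skydiving.
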AnimBlueprint /BRCosmetics/Characters/Player/Female/Medium/Bodies/F_MED_Sarong/Meshes/Parts/F_MED_Sarong_FaceAcc_AnimBP
target_skeleton=&/F_MED_Sarong_FaceAcc_Skeleton
graphs [
 AnimGraph {nodes [
   n1 = "Constraint - dyn_ponytail_03 (PP)"!(ComponentPose=Local To Component(LocalPose=Use cached pose 'lod0Cache'()), Alpha=1, ConstraintWeights_0=0.5, ConstraintWeights_1:float=0.5) -> (Pose) {
    alphaBoolBlend {blendOption=Linear}
    boneToModify {boneName=dyn_ponytail_03}
    constraintSetup [{targetBone {boneName=spine_05}}, {targetBone {boneName=head}}]
    constraintWeights [1,1]
   }
   n2 = "Constraint - dyn_ponytail_04 (PP)"!(ComponentPose=@n1, Alpha=1, ConstraintWeights_0=0.5, ConstraintWeights_1:float=0.5) -> (Pose) {
    alphaBoolBlend {blendOption=Linear}
    boneToModify {boneName=dyn_ponytail_04}
    constraintSetup [{targetBone {boneName=spine_02}}, {targetBone {boneName=head}}]
    constraintWeights [1,1]
   }
   n3 = "Constraint - dyn_ponytail_05 (PP)"!(ComponentPose=@n2, Alpha=1, ConstraintWeights_0=0.5, ConstraintWeights_1:float=0.5) -> (Pose) {
    alphaBoolBlend {blendOption=Linear}
    boneToModify {boneName=dyn_ponytail_05}
    constraintSetup [{targetBone {boneName=spine_02}}, {targetBone {boneName=head}}]
    constraintWeights [1,1]
   }
   n4 = "Transform (Modify) Bone - Bone: dyn_ponytail_01"(ComponentPose=@n3, Alpha=Get ponytailSwingAlpha() -> (ponytailSwingAlpha:double), Translation="0,0,0", Rotation="0,35,0", Scale="1,1,1") {
    rotationMode=BMM_Additive
    rotationSpace=BCS_BoneSpace
    translationMode=BMM_Additive
    translationSpace=BCS_BoneSpace
    alphaBoolBlend {blendOption=Linear}
    boneToModify {boneName=dyn_ponytail_01}
    rotation {yaw=35}
   }
   Output Pose(Result=Blend Poses by bool(
     BlendPose_0=Component To Local(ComponentPose=@n4)
     BlendPose_1=Use cached pose 'lod0Cache'()
     bActiveValue=Get isMobileLod() -> (isMobileLod:bool)
    ) {blendType=Linear}) {layerGroup=None}
   n5 = Anim Dynamics - Bone: dyn_flowerpedalA_01_l!(ComponentPose=Local To Component(LocalPose=Copy Pose From Mesh() {bUseAttachedParent=true}), ExternalForce="0,0,0") -> (Pose) {
    alpha=0.75
    angularDampingOverride=1
    angularSpringConstant=3500
    bAngularSpring=true
    bGravityOverrideInSimSpace=true
    bOverrideAngularDamping=true
    bUseGravityOverride=true
    gravityScale=0.15
    simulationSpace=RootRelative
    alphaBoolBlend {blendOption=Linear}
    boundBone {boneName=dyn_flowerpedalA_01_l}
    componentAppliedLinearAccClamp=<3000 3000 3000>
    componentLinearAccScale=<0.2 0.2 0.2>
    componentLinearVelScale=<0.2 0.2 0.2>
    gravityOverride {z=-500}
    physicsBodyDefinitions [{
      sphereCollisionRadius=1.5
      boundBone {boneName=dyn_flowerpedalA_01_l}
      boxExtents=<10 10 10>
      constraintSetup {angularLimitsMax=<0 10 0>, angularLimitsMin=<0 -10 -10>, angularTarget=<1 0 0>}
      localJointOffset=<1 0 0>
     }]
   }
   n6 = Anim Dynamics - Bone: dyn_flowerpedalA_02_l!(ComponentPose=@n5, ExternalForce="0,0,0") -> (Pose) {
    alpha=0.75
    angularDampingOverride=1
    angularSpringConstant=3500
    bAngularSpring=true
    bGravityOverrideInSimSpace=true
    bOverrideAngularDamping=true
    bUseGravityOverride=true
    gravityScale=0.15
    simulationSpace=RootRelative
    alphaBoolBlend {blendOption=Linear}
    boundBone {boneName=dyn_flowerpedalA_02_l}
    componentAppliedLinearAccClamp=<3000 3000 3000>
    componentLinearAccScale=<0.2 0.2 0.2>
    componentLinearVelScale=<0.2 0.2 0.2>
    gravityOverride {z=-500}
    physicsBodyDefinitions [{
      sphereCollisionRadius=1.5
      boundBone {boneName=dyn_flowerpedalA_02_l}
      boxExtents=<10 10 10>
      constraintSetup {angularLimitsMax=<0 10 0>, angularLimitsMin=<0 -10 -10>, angularTarget=<1 0 0>}
      localJointOffset=<1 0 0>
     }]
   }
   n7 = Anim Dynamics - Bone: dyn_flowerpedalB_01_l!(ComponentPose=@n6, ExternalForce="0,0,0") -> (Pose) {
    alpha=0.75
    angularDampingOverride=1
    angularSpringConstant=3500
    bAngularSpring=true
    bGravityOverrideInSimSpace=true
    bOverrideAngularDamping=true
    bUseGravityOverride=true
    gravityScale=0.15
    simulationSpace=RootRelative
    alphaBoolBlend {blendOption=Linear}
    boundBone {boneName=dyn_flowerpedalB_01_l}
    componentAppliedLinearAccClamp=<3000 3000 3000>
    componentLinearAccScale=<0.2 0.2 0.2>
    componentLinearVelScale=<0.2 0.2 0.2>
    gravityOverride {z=-500}
    physicsBodyDefinitions [{
      sphereCollisionRadius=1.5
      boundBone {boneName=dyn_flowerpedalB_01_l}
      boxExtents=<10 10 10>
      constraintSetup {angularLimitsMax=<0 10 0>, angularLimitsMin=<0 -10 -10>, angularTarget=<1 0 0>}
      localJointOffset=<1 0 0>
     }]
   }
   n8 = Anim Dynamics - Bone: dyn_flowerpedalB_02_l!(ComponentPose=@n7, ExternalForce="0,0,0") -> (Pose) {
    alpha=0.75
    angularDampingOverride=1
    angularSpringConstant=3500
    bAngularSpring=true
    bGravityOverrideInSimSpace=true
    bOverrideAngularDamping=true
    bUseGravityOverride=true
    gravityScale=0.15
    simulationSpace=RootRelative
    alphaBoolBlend {blendOption=Linear}
    boundBone {boneName=dyn_flowerpedalB_02_l}
    componentAppliedLinearAccClamp=<3000 3000 3000>
    componentLinearAccScale=<0.2 0.2 0.2>
    componentLinearVelScale=<0.2 0.2 0.2>
    gravityOverride {z=-500}
    physicsBodyDefinitions [{
      sphereCollisionRadius=1.5
      boundBone {boneName=dyn_flowerpedalB_02_l}
      boxExtents=<10 10 10>
      constraintSetup {angularLimitsMax=<0 10 0>, angularLimitsMin=<0 -10 -10>, angularTarget=<1 0 0>}
      localJointOffset=<1 0 0>
     }]
   }
   n9 = Anim Dynamics - Bone: dyn_flowerpedalC_01_l!(ComponentPose=@n8, ExternalForce="0,0,0") -> (Pose) {
    alpha=0.75
    angularDampingOverride=1
    angularSpringConstant=3500
    bAngularSpring=true
    bGravityOverrideInSimSpace=true
    bOverrideAngularDamping=true
    bUseGravityOverride=true
    gravityScale=0.15
    simulationSpace=RootRelative
    alphaBoolBlend {blendOption=Linear}
    boundBone {boneName=dyn_flowerpedalC_01_l}
    componentAppliedLinearAccClamp=<3000 3000 3000>
    componentLinearAccScale=<0.2 0.2 0.2>
    componentLinearVelScale=<0.2 0.2 0.2>
    gravityOverride {z=-500}
    physicsBodyDefinitions [{
      sphereCollisionRadius=1.5
      boundBone {boneName=dyn_flowerpedalC_01_l}
      boxExtents=<10 10 10>
      constraintSetup {angularLimitsMax=<0 10 0>, angularLimitsMin=<0 -10 -10>, angularTarget=<1 0 0>}
      localJointOffset=<1 0 0>
     }]
   }
   n10 = Anim Dynamics - Bone: dyn_flowerpedalC_02_l!(ComponentPose=@n9, ExternalForce="0,0,0") -> (Pose) {
    alpha=0.75
    angularDampingOverride=1
    angularSpringConstant=3500
    bAngularSpring=true
    bGravityOverrideInSimSpace=true
    bOverrideAngularDamping=true
    bUseGravityOverride=true
    gravityScale=0.15
    simulationSpace=RootRelative
    alphaBoolBlend {blendOption=Linear}
    boundBone {boneName=dyn_flowerpedalC_02_l}
    componentAppliedLinearAccClamp=<3000 3000 3000>
    componentLinearAccScale=<0.2 0.2 0.2>
    componentLinearVelScale=<0.2 0.2 0.2>
    gravityOverride {z=-500}
    physicsBodyDefinitions [{
      sphereCollisionRadius=1.5
      boundBone {boneName=dyn_flowerpedalC_02_l}
      boxExtents=<10 10 10>
      constraintSetup {angularLimitsMax=<0 10 0>, angularLimitsMin=<0 -10 -10>, angularTarget=<1 0 0>}
      localJointOffset=<1 0 0>
     }]
   }
   n11 = Anim Dynamics - Bone: dyn_flowerpedalD_01_l!(ComponentPose=@n10, ExternalForce="0,0,0") -> (Pose) {
    alpha=0.75
    angularDampingOverride=1
    angularSpringConstant=3500
    bAngularSpring=true
    bGravityOverrideInSimSpace=true
    bOverrideAngularDamping=true
    bUseGravityOverride=true
    gravityScale=0.15
    simulationSpace=RootRelative
    alphaBoolBlend {blendOption=Linear}
    boundBone {boneName=dyn_flowerpedalD_01_l}
    componentAppliedLinearAccClamp=<3000 3000 3000>
    componentLinearAccScale=<0.2 0.2 0.2>
    componentLinearVelScale=<0.2 0.2 0.2>
    gravityOverride {z=-500}
    physicsBodyDefinitions [{
      sphereCollisionRadius=1.5
      boundBone {boneName=dyn_flowerpedalD_01_l}
      boxExtents=<10 10 10>
      constraintSetup {angularLimitsMax=<0 10 0>, angularLimitsMin=<0 -10 -10>, angularTarget=<1 0 0>}
      localJointOffset=<1 0 0>
     }]
   }
   n12 = Anim Dynamics - Bone: dyn_flowerpedalD_02_l!(ComponentPose=@n11, ExternalForce="0,0,0") -> (Pose) {
    alpha=0.75
    angularDampingOverride=1
    angularSpringConstant=3500
    bAngularSpring=true
    bGravityOverrideInSimSpace=true
    bOverrideAngularDamping=true
    bUseGravityOverride=true
    gravityScale=0.15
    simulationSpace=RootRelative
    alphaBoolBlend {blendOption=Linear}
    boundBone {boneName=dyn_flowerpedalD_02_l}
    componentAppliedLinearAccClamp=<3000 3000 3000>
    componentLinearAccScale=<0.2 0.2 0.2>
    componentLinearVelScale=<0.2 0.2 0.2>
    gravityOverride {z=-500}
    physicsBodyDefinitions [{
      sphereCollisionRadius=1.5
      boundBone {boneName=dyn_flowerpedalD_02_l}
      boxExtents=<10 10 10>
      constraintSetup {angularLimitsMax=<0 10 0>, angularLimitsMin=<0 -10 -10>, angularTarget=<1 0 0>}
      localJointOffset=<1 0 0>
     }]
   }
   n13 = Anim Dynamics - Bone: dyn_flowerpedalE_01_l!(ComponentPose=@n12, ExternalForce="0,0,0") -> (Pose) {
    alpha=0.75
    angularDampingOverride=1
    angularSpringConstant=3500
    bAngularSpring=true
    bGravityOverrideInSimSpace=true
    bOverrideAngularDamping=true
    bUseGravityOverride=true
    gravityScale=0.15
    simulationSpace=RootRelative
    alphaBoolBlend {blendOption=Linear}
    boundBone {boneName=dyn_flowerpedalE_01_l}
    componentAppliedLinearAccClamp=<3000 3000 3000>
    componentLinearAccScale=<0.2 0.2 0.2>
    componentLinearVelScale=<0.2 0.2 0.2>
    gravityOverride {z=-500}
    physicsBodyDefinitions [{
      sphereCollisionRadius=1.5
      boundBone {boneName=dyn_flowerpedalE_01_l}
      boxExtents=<10 10 10>
      constraintSetup {angularLimitsMax=<0 10 0>, angularLimitsMin=<0 -10 -10>, angularTarget=<1 0 0>}
      localJointOffset=<1 0 0>
     }]
   }
   n14 = Anim Dynamics - Bone: dyn_flowerpedalE_02_l!(ComponentPose=@n13, ExternalForce="0,0,0") -> (Pose) {
    alpha=0.75
    angularDampingOverride=1
    angularSpringConstant=3500
    bAngularSpring=true
    bGravityOverrideInSimSpace=true
    bOverrideAngularDamping=true
    bUseGravityOverride=true
    gravityScale=0.15
    simulationSpace=RootRelative
    alphaBoolBlend {blendOption=Linear}
    boundBone {boneName=dyn_flowerpedalE_02_l}
    componentAppliedLinearAccClamp=<3000 3000 3000>
    componentLinearAccScale=<0.2 0.2 0.2>
    componentLinearVelScale=<0.2 0.2 0.2>
    gravityOverride {z=-500}
    physicsBodyDefinitions [{
      sphereCollisionRadius=1.5
      boundBone {boneName=dyn_flowerpedalE_02_l}
      boxExtents=<10 10 10>
      constraintSetup {angularLimitsMax=<0 10 0>, angularLimitsMin=<0 -10 -10>, angularTarget=<1 0 0>}
      localJointOffset=<1 0 0>
     }]
   }
   n15 = Anim Dynamics - Bone: dyn_flower_center_02!(ComponentPose=@n14, ExternalForce="0,0,0") -> (Pose) {
    alpha=0.75
    angularDampingOverride=1
    angularSpringConstant=3500
    bAngularSpring=true
    bGravityOverrideInSimSpace=true
    bOverrideAngularDamping=true
    bUseGravityOverride=true
    gravityScale=0.15
    simulationSpace=RootRelative
    alphaBoolBlend {blendOption=Linear}
    boundBone {boneName=dyn_flower_center_02}
    componentAppliedLinearAccClamp=<3000 3000 3000>
    componentLinearAccScale=<0.2 0.2 0.2>
    componentLinearVelScale=<0.2 0.2 0.2>
    gravityOverride {z=-500}
    physicsBodyDefinitions [{
      sphereCollisionRadius=1.5
      boundBone {boneName=dyn_flower_center_02}
      boxExtents=<10 10 10>
      constraintSetup {angularLimitsMax=<0 10 0>, angularLimitsMin=<0 -10 -10>, angularTarget=<1 0 0>}
      localJointOffset=<1 0 0>
     }]
   }
   n16 = Anim Dynamics - Bone: dyn_flower_center_01!(ComponentPose=@n15, ExternalForce="0,0,0") -> (Pose) {
    alpha=0.75
    angularDampingOverride=1
    angularSpringConstant=3500
    bAngularSpring=true
    bGravityOverrideInSimSpace=true
    bOverrideAngularDamping=true
    bUseGravityOverride=true
    gravityScale=0.15
    simulationSpace=RootRelative
    alphaBoolBlend {blendOption=Linear}
    boundBone {boneName=dyn_flower_center_01}
    componentAppliedLinearAccClamp=<3000 3000 3000>
    componentLinearAccScale=<0.2 0.2 0.2>
    componentLinearVelScale=<0.2 0.2 0.2>
    gravityOverride {z=-500}
    physicsBodyDefinitions [{
      sphereCollisionRadius=1.5
      boundBone {boneName=dyn_flower_center_01}
      boxExtents=<10 10 10>
      constraintSetup {angularLimitsMax=<0 10 0>, angularLimitsMin=<0 -10 -10>, angularTarget=<1 0 0>}
      localJointOffset=<1 0 0>
     }]
   }
   n17 = RigidBody(ComponentPose=@n16) {
    bClampLinearTranslationLimitToRefPose=true
    bForceDisableCollisionBetweenConstraintBodies=true
    bOverrideWorldGravity=true
    overridePhysicsAsset=&/F_MED_Sarong_FaceAcc_Physics
    simulationSpace=BaseBoneSpace
    alphaBoolBlend {blendOption=Linear}
    baseBoneRef {boneName=neck_02}
    componentAppliedLinearAccClamp=<3000 3000 3000>
    componentLinearAccScale=<0.3 0.3 0.3>
    componentLinearVelScale=<0.7 0.7 0.7>
    overrideWorldGravity {z=-980}
   }
   Save cached pose 'lod0Cache'(Pose=Component To Local(ComponentPose=@n17)) {cacheName=lod0Cache}
  ]}
 EventGraph {nodes [
   Event BlueprintUpdateAnimation()
   n1 = Get bIsTargeting()
   > Set aimAlphaInv(aimAlphaInv:double="To Float (Boolean)"(InBool=NOT Boolean(A=@n1))) -> (Output_Get:double)
   n2 = "To Float (Boolean)"(InBool=Get bIsSkyDiving())
   > Set skyDivingExtVector(skyDivingExtVector:Vector="To Vector (Float)"!(InDouble:double=MapRangeClamped(Value=@n2, InRangeA=0, InRangeB=1, OutRangeA=1.5, OutRangeB=7)) -> (ReturnValue:Vector)) -> (Output_Get:Vector)
   n3 = Get CurrentLOD()
   n4 = Get bRigidBodyEnabled()
   > Set isMobileLod(isMobileLod:bool=OR Boolean(A=NOT Boolean(A=@n4), B="integer > integer"(A=@n3, B=1) {is=title K2Node_CallFunction})) -> (Output_Get:bool)
   n5 = GetTwistAndSwingAngleOfDeltaRotationFromRefPose(self=GetOwnerSkeletalMeshComponent(), BoneName=head)
   > Set ponytailSwingAlpha(ponytailSwingAlpha:double=SelectFloat(A=1, B=MapRangeClamped(Value=@n5.OutSwingAngle, InRangeA=5, InRangeB=50, OutRangeB=0.5), bPickA=Get bIsSkyDiving())) -> (Output_Get:double)
  ]}
]
variables [
 aimAlphaInv {type=number}
 isMobileLod {type=boolean}
 ponytailSwingAlpha {type=number}
 skyDivingExtVector {type=Vector}
]
